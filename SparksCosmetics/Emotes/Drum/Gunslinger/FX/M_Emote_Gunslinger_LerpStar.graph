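# Unlit masked material for Niagara sprites in the Gunslinger drum emote FX, drawing an outlined star.
# A star SDF texture is thresholded by Outline Thickness to blend from Outline Color to the particle color.
# The same texture, scaled by Opacity Control and saturated, multiplies particle alpha to form the mask.
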Material /SparksCosmetics/Emotes/Drum/Gunslinger/FX/M_Emote_Gunslinger_LerpStar
graph {
 properties {bAutomaticallySetUsageInEditor=false, bUsedWithNiagaraSprites=true, blendMode=BLEND_Masked, shadingModel=MSM_Unlit, usedShadingModels=Unlit}
 nodes [
  n1 = ParticleColor()
  n2 = TextureSampleParameter2D() {
   parameterName=Shape
   samplerType=SAMPLERTYPE_LinearGrayscale
   texture=/Game/Accessories/FORT_Backpacks/Backpack_F_MED_Cactus_Dancer/Textures/starSDF_6
  }
  LinearInterpolate(
   A=VectorParameter() {parameterName=Outline Color, defaultValue=<rgba 0.0203 0.0137 0.0152 1>}.RGB
   B=@n1.RGB
   Alpha=MaterialFunctionCall!(
    ThresholdValue=ScalarParameter() {defaultValue=0.3, parameterName=Outline Thickness}
    ThresholdRange=Constant()
    Alpha=@n2.R
    TwoSidedThreshold
   ) {materialFunction=/Engine/Functions/Engine_MaterialFunctions02/Masking/ThresholdWithRange}.Result
  )
  n3 = ScalarParameter() {defaultValue=20, parameterName=Opacity Control}
  Multiply(A=@n1.A, B=Saturate(None=Multiply(A=@n2.R, B=@n3) {constB=15}))
 ]
}
