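# Niagara system for the Gunslinger drum emote, drawing a grey smoke wisp as a single ribbon emitter.
# User.NormalizedNotifyProgress drives spawn rate, lifetime and ribbon width scale, so the trail tapers off as the notify ends.
# System age fades in initial alpha and width, while per-particle curves shape both over life.
# Particles rise and drift through added velocity, User.Wind, local acceleration and curl noise that strengthens with emitter age.
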NiagaraSystem /SparksCosmetics/Emotes/Drum/Gunslinger/FX/NS_Emote_Gunslinger_Smoke
dependencies {usedDataInterfaces [&NiagaraDataInterfaceCurve], usedDynamicInputs [], usedModules [&SystemState]}
emitters {Smoke {
  emitterUpdateScript [
   &EmitterState()
   &SpawnRate(
    SpawnRate=&FloatFromCurve(
     FloatCurve=DataInterface() {propertyValues {Curve {keys [
         {interpMode=RCIM_Cubic, leaveTangent=-5.3156, tangentMode=RCTM_Break, time=0.6, value=1}
         {arriveTangent=-5.2763, interpMode=RCIM_Cubic, tangentMode=RCTM_Break, time=1, value=0.05}
        ]}}}
     CurveIndex=$User.NormalizedNotifyProgress
     Scale Curve=25
    )
    Use Spawn Probability=false
   )
  ]
  particleSpawnScript [
   &InitializeParticle(
    Color=&MakeLinearColorFromVectorAndFloat(
     "Vector (RGB)"=&MakeVectorFromLinearColorRGB(Color=<rgba 0.3125 0.3125 0.3125 1>)
     "Float (Alpha)"=&FloatFromCurve(
      FloatCurve=DataInterface() {propertyValues {Curve {keys [{interpMode=RCIM_Cubic, tangentMode=RCTM_Break, time=0, value=0}, {interpMode=RCIM_Cubic, tangentMode=RCTM_Break, time=0.2, value=1}]}}}
      CurveIndex=$System.Age
     )
    )
    Lifetime=&FloatFromCurve(
     FloatCurve=DataInterface() {propertyValues {Curve {keys [
         {interpMode=RCIM_Cubic, tangentMode=RCTM_Break, time=0, value=0.425}
         {arriveTangent=-0.0092, interpMode=RCIM_Cubic, tangentMode=RCTM_Break, time=0.2045, value=1}
        ]}}}
     CurveIndex=$User.NormalizedNotifyProgress
     Scale Curve=0.8
    )
    Ribbon Width=&FloatFromCurve(
     FloatCurve=DataInterface() {propertyValues {Curve {keys [{interpMode=RCIM_Cubic, tangentMode=RCTM_Break, time=0, value=0}, {interpMode=RCIM_Cubic, tangentMode=RCTM_Break, time=0.2, value=1}]}}}
     CurveIndex=$System.Age
     Scale Curve=15
    )
    Ribbon Width Mode=::NewEnumerator1
    UsePositionOffset=false
    Write Material Random=true
   )
   &AddVelocity(Velocity=<0 0 25>)
   &AddVelocity(Velocity=&MakeVector(
     X=&Multiply_Float(A=$User.Wind, B=100)
     Y=&Multiply_Float(A=$User.Wind, B=-200)
     Z=0
    )) {module=AddVelocity001}
  ]
  particleUpdateScript [
   &CurlNoiseForce(
    Noise Strength=&FloatFromCurve(
     FloatCurve=DataInterface() {propertyValues {Curve {keys [
         {interpMode=RCIM_Cubic, leaveTangent=0.0121, tangentMode=RCTM_Break, time=0, value=0}
         {interpMode=RCIM_Cubic, tangentMode=RCTM_Break, time=0.4, value=1}
        ]}}}
     CurveIndex=$Emitter.Age
     Scale Curve=100
    )
    Pan Noise=false
   )
   &AccelerationForce(Acceleration=&Multiply_VectorByFloat(Vector=<0 1 1>, Float=1), Coordinate Space=::Local)
   &SolveForcesAndVelocity(Clamp Velocity=false, Limit Acceleration=false)
   &ParticleState()
   &ScaleColor(
    Scale Alpha=&FloatFromCurve(FloatCurve=DataInterface() {propertyValues {Curve {keys [
         <at 0 0 RCIM_Cubic>
         {interpMode=RCIM_Cubic, tangentMode=RCTM_Break, time=0.2, value=1}
         {arriveTangent=-0.262, interpMode=RCIM_Cubic, tangentMode=RCTM_Break, time=1, value=0}
        ]}}})
    ScaleA=true
    ScaleRGB=true
   )
   &ScaleRibbonWidth(Ribbon Width Scale=&FloatFromCurve(
     FloatCurve=DataInterface() {propertyValues {Curve {keys [
         {interpMode=RCIM_Cubic, tangentMode=RCTM_Break, time=0, value=0}
         {interpMode=RCIM_Cubic, tangentMode=RCTM_Break, time=0.2, value=1}
         {interpMode=RCIM_Cubic, tangentMode=RCTM_Break, time=0.6, value=1}
         {arriveTangent=-1.7378, interpMode=RCIM_Cubic, tangentMode=RCTM_Break, time=1, value=0}
        ]}}}
     CurveIndex=$User.NormalizedNotifyProgress
     Scale Curve=&FloatFromCurve(FloatCurve=DataInterface() {propertyValues {Curve {keys [
          {arriveTangent=0.0313, interpMode=RCIM_Cubic, leaveTangent=0.0313, tangentMode=RCTM_User, time=0, value=0}
          {interpMode=RCIM_Cubic, tangentMode=RCTM_User, time=0.5, value=1}
          {arriveTangent=-2.9653, tangentMode=RCTM_User, time=1, value=0}
         ]}}})
    ))
  ]
  rendererClasses [&NiagaraRibbonRendererProperties]
  renderers [{rendererClass=&NiagaraRibbonRendererProperties}]
 }}
summary {
 systemName=NS_Emote_Gunslinger_Smoke
 emitters [{emitterName=Smoke, rendererClasses [&NiagaraRibbonRendererProperties]}]
 userVariables [
  User.IsFrontEnd {type=NiagaraBool, defaultValue=true}
  User.IsLargeBody {type=NiagaraBool}
  User.NormalizedNotifyProgress {type=NiagaraFloat, defaultValue=0, description=""}
  User.Wind {type=NiagaraFloat, defaultValue=0, description=""}
 ]
}
